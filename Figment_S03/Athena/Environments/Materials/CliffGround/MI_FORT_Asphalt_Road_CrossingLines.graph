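# Deferred decal material for road crossing lines on asphalt, using translucent blending and default lit shading.
# The graph holds one texture parameter sampling a concrete crosswalk mask, plus three scalar constants.
# Node connections and output assignments are not shown, so how the mask and constants drive the decal is not visible.
# Despite its MI_ prefix, the asset is a base Material rather than a material instance.
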Material /Figment_S03/Athena/Environments/Materials/CliffGround/MI_FORT_Asphalt_Road_CrossingLines
graph {
 properties {blendMode=BLEND_Translucent, materialDomain=MD_DeferredDecal, pixelDepthOffsetMode=PDOM_Legacy, usedShadingModels=Default Lit}
 nodes [
  Constant() {r=0.65}
  Constant()
  Constant() {r=0.5}
  TextureSampleParameter2D() {texture=/Game/Packages/DS_Fortnight/T_FRT_Conc_Masks_Crosswalks}
 ]
}
