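# Material function for event wetness on material attributes; its output connects straight to the input, leaving the wet graph unused.
# That graph tints BaseColor by WetTone, slightly more flatly on up-facing normals, and lowers Roughness.
# World height relative to a clamped water level, from the terrain MPC or a parameter, fades wetness out upward.
# DisableWet defaults on, and the unconnected shading-path switch bypasses the effect on mobile.
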MaterialFunction /Game/Athena/WinterArt/Materials/MF_Event_Wet
graph {nodes [
  n1 = FunctionInput() {inputType=FunctionInput_MaterialAttributes}
  FunctionOutput(None=@n1)
  n2 = GetMaterialAttributes(None=@n1) {attributeGetTypes [#1, #2]}
  n3 = VectorParameter() {group=Wet, parameterName=WetTone, defaultValue=<rgba 0.4635 0.3603 0.2412 0>}
  n4 = VertexNormalWS()
  n5 = LinearInterpolate(
   A=Multiply(A=@n2.BaseColor, B=@n3.RGB)
   B=@n3.RGB
   Alpha=LinearInterpolate(B=Constant() {r=0.1}, Alpha=ComponentMask(None=@n4) {b=true}) {constB=0.35}
  ) {constAlpha=0.2}
  n6 = Multiply(A=@n2.Roughness, B=Constant() {r=0.35})
  n7 = ComponentMask(None=WorldPosition().XYZ) {b=true}
  n8 = CollectionParameter() {collection=/Game/Athena/Apollo/Environments/Landscape/MPC/Apollo_MPC_Terrain, parameterName=WaterLevel}
  n9 = ScalarParameter() {defaultValue=-3500, group=WorldWater, parameterName=WaterHeight}
  n10 = MaterialProxyReplace(
   Realtime=Clamp(None=StaticSwitchParameter(True=@n8, False=@n9) {defaultValue=true, group=WorldWater, parameterName=UseMPC}) {maxDefault=-1981, minDefault=-4669}
   MaterialProxy=Constant()
  )
  n11 = Divide(A=Add(A=Subtract(A=@n7, B=@n10), B=Constant() {r=-768})) {constB=500}
  n12 = StaticSwitchParameter(
   True=@n1
   False=BlendMaterialAttributes(
    A=SetMaterialAttributes!(MaterialAttributes=@n2.MaterialAttributes, BaseColor=@n5, Roughness=@n6) {attributeSetTypes [#1, #2]}
    B=@n1
    Alpha=Saturate(None=@n11)
   )
  ) {defaultValue=true, group=Wet, parameterName=DisableWet}
  ShadingPathSwitch(Default=@n12, Deferred=@n12, Forward=@n12, Mobile=@n1)
 ]}
