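# Master terrain material for cliff and ground meshes, with switchable grass or base, mud, concrete, dirt, top, cliff and paint layers.
# Mostly world-aligned textures are height-blended by vertex color: red mud/concrete, green top layer or second cliff, blue dirt, alpha cliffs.
# Texture-masked paint goes over these, then event wetness, optional snow and fresnel.
# Mobile and low-quality fallbacks, border masking, LOD offset, Creative move-tool and Curie landscape functions finish it.
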Material /Figment_S03/Athena/Environments/Materials/CliffGround/Terrain_Mat
graph {
 properties {
  bTangentSpaceNormal=false
  bUseMaterialAttributes=true
  bUsedWithInstancedStaticMeshes=true
  bUsedWithMeshDeformer=true
  bUsedWithNanite=true
  bUsedWithSkeletalMesh=true
  bUsedWithSplineMeshes=true
  ditheredLODTransition=true
  physMaterial=/Game/FortressPhysicalMaterials/Dirt
  pixelDepthOffsetMode=PDOM_Legacy
  refractionMethod=RM_IndexOfRefraction
  usedShadingModels=Default Lit
 }
 nodes [
  Constant()
  n1 = VectorParameter() {group=Snow, parameterName=SnowColor_1, defaultValue=<rgba 0.7262 0.7933 0.83 1>}
  n2 = ScalarParameter() {defaultValue=-0.0381, group=Snow, parameterName=SnowNormalContrast1}
  n3 = ScalarParameter() {defaultValue=1.4108, group=Snow, parameterName=SnowNormalContrast2}
  n4 = TextureCoordinate()
  n5 = VectorParameter() {group=PaintLayer, parameterName=PaintLayerUV_RG_Offset_BA_Scale, defaultValue=<rgba 0 0 1 1>}
  n6 = ComponentMask(None=@n5.RGB) {g=true, r=true}
  n7 = TextureSampleParameter2D(UVs=Multiply(A=Add(A=@n4, B=@n6), B=AppendVector(A=@n5.B, B=@n5.B))) {
   group=PaintLayer
   parameterName=PaintLayerMasks
   samplerType=SAMPLERTYPE_Masks
   texture=/Game/Packages/DS_Fortnight/SM/Textures/T_SoccerTennisBasketball
  }
  n8 = StaticComponentMaskParameter(None=AppendVector(A=@n7.RGB, B=@n7.RGB)) {defaultR=true, group=PaintLayer, parameterName=PaintLayerMaskChannel}
  n9 = WorldPosition()
  n10 = TextureSampleParameter2D(UVs=Divide(
    A=ComponentMask(None=@n9.XYZ) {g=true, r=true}
    B=ScalarParameter() {defaultValue=2048, group=PaintLayer, parameterName=PaintLayerMask_Mask Tiling}
   ) {constB=1024}) {
   group=PaintLayer
   parameterName=PaintLayerMasks_Mask
   samplerType=SAMPLERTYPE_Masks
   texture=/Game/Packages/Fortnite_UniqueMaterials_1/Textures/T_FRT_Conc_Masks_B_new
  }
  n11 = AppendVector(A=@n10.RGB, B=@n10.RGB)
  n12 = StaticSwitchParameter(
   True=Multiply(A=@n8, B=StaticComponentMaskParameter(None=@n11) {defaultR=true, group=PaintLayer, parameterName=PaintLayerMask_Mask_Channel})
   False=@n8
  ) {desc=Diffuse, group=PaintLayer, parameterName=Use Second PaintLayer Mask}
  n13 = ScalarParameter() {defaultValue=1, group=PaintLayer, parameterName=PaintLayer Alpha Modulation}
  n14 = TextureObjectParameter() {
   group=Cliffs
   parameterName=CliffTexture
   samplerSource=SSM_Wrap_WorldGroupSettings
   texture=/Game/Packages/DS_Fortnite_Terrain_NoLOD/Textures/T_Cave_Exterior_Wall_D
  }
  n15 = TextureObjectParameter() {
   group=Cliffs
   parameterName=Cliff Spec Texture
   samplerSource=SSM_Wrap_WorldGroupSettings
   texture=/Game/Packages/DS_Fortnite_Terrain_NoLOD/Textures/T_Cave_Exterior_Wall_S
  }
  n16 = StaticBool() {value=true}
  n17 = TextureObjectParameter() {
   group=Cliffs
   parameterName=CliffNormal
   samplerSource=SSM_Wrap_WorldGroupSettings
   samplerType=SAMPLERTYPE_Normal
   texture=/Game/Packages/DS_Fortnite_Terrain_NoLOD/Textures/T_Cave_Exterior_Wall_N
  }
  n18 = WorldPosition()
  n19 = ComponentMask(None=@n18.XYZ) {g=true, r=true}
  n20 = ComponentMask(None=@n18.XYZ) {b=true}
  n21 = Multiply(A=ScalarParameter() {defaultValue=0.25, group=Cliffs, parameterName=XY coordinate offset over Z}, B=@n20) {constB=0.25}
  n22 = Multiply(
   A=MaterialFunctionCall(f2=@n19) {materialFunction=&AddComponents}.f2
   B=ScalarParameter() {defaultValue=0.25, group=Cliffs, parameterName=Z coordinate offset over XY}
  ) {constB=0.25}
  n23 = MaterialFunctionCall!(
   Diffuse Texture Object=@n14
   Spec Texture Object=@n15
   TextureSize=Abs(None=ScalarParameter() {defaultValue=1300, group=Cliffs, parameterName=CliffWorldCoordScale})
   --------------------------
   Use seperate settings for Z projection
   Use seperate texture for Z projection
   Texture Size Z projection
   Z Plane Diffuse Texture Object
   Z Plane Spec Texture Object
   "--------------------------      "
   ProjectionTransitionContrast=ScalarParameter() {group=Cliffs, parameterName=Cliff Projection Contrast}
   Use Texture Alpha Heightmap to Refine the Transistion=@n16
   HeightLerpTransitionContrast=ScalarParameter() {defaultValue=3, group=Cliffs, parameterName=Cliff Height Lerp Contrast}
   PickTheDiffuseHeightChannel
   TheDiffuseHeightChannel
   "--------------------------       "
   NormalTextureObject=@n17
   World Space Normals=@n16
   "--------------------------  "
   Export Float 4 Diffuse=@n16
   Export Float 4 Spec
   "--------------------------     "
   WorldPosition=AppendVector(A=Add(A=@n19, B=@n21), B=Add(A=@n22, B=@n20))
   World Space Normal
  ) {materialFunction=/Engine/Functions/Engine_MaterialFunctions02/Texturing/WorldAlignedTextures_Complex}
  n24 = MaterialFunctionCall!(In=StaticSwitchParameter(True=@n23.XYZ Mat Attributes - See Tooltip, False=@n23.XYZ Mat Attributes - See Tooltip) {defaultValue=true, group=Cliffs, parameterName="Use XYZ Cliff Projection. False=XY"}) {materialFunction=/Game/Packages/Fortress_SharedMaterials/Base_Material_Functions/SwizzleWorldAlignedSpecOutputForFortnite}
  n25 = StaticBoolParameter() {defaultValue=true, group=Cliffs, parameterName=Use Cliff Spec}
  n26 = StaticBoolParameter() {defaultValue=true, group=Cliffs, parameterName=Use Cliff Normals}
  n27 = VertexColor()
  n28 = TextureObjectParameter() {
   group=Cliffs_2
   parameterName=CliffTexture_2
   samplerSource=SSM_Wrap_WorldGroupSettings
   texture=&T_Cave_Exterior_Wall_D
  }
  n29 = TextureObjectParameter() {
   group=Cliffs_2
   parameterName=Cliff Spec Texture_2
   samplerSource=SSM_Wrap_WorldGroupSettings
   texture=&T_Cave_Exterior_Wall_S
  }
  n30 = StaticBool() {value=true}
  n31 = TextureObjectParameter() {
   group=Cliffs_2
   parameterName=CliffNormal_2
   samplerSource=SSM_Wrap_WorldGroupSettings
   samplerType=SAMPLERTYPE_Normal
   texture=&FlatNormalBlue
  }
  n32 = WorldPosition()
  n33 = ComponentMask(None=@n32.XYZ) {b=true}
  n34 = Multiply(A=ScalarParameter() {defaultValue=0.25, group=Cliffs_2, parameterName=XY coordinate offset over Z_2}, B=@n33) {constB=0.25}
  n35 = ComponentMask(None=@n32.XYZ) {g=true, r=true}
  n36 = Multiply(
   A=MaterialFunctionCall(f2=@n35) {materialFunction=&AddComponents}.f2
   B=ScalarParameter() {defaultValue=0.25, group=Cliffs_2, parameterName=Z coordinate offset over XY_2}
  ) {constB=0.25}
  n37 = MaterialFunctionCall(
   Diffuse Texture Object=@n28
   Spec Texture Object=@n29
   TextureSize=ScalarParameter() {defaultValue=-768, group=Cliffs_2, parameterName=CliffWorldCoordScale_2}
   ProjectionTransitionContrast=ScalarParameter() {group=Cliffs_2, parameterName=Cliff Projection Contrast_2}
   Use Texture Alpha Heightmap to Refine the Transistion=@n30
   HeightLerpTransitionContrast=ScalarParameter() {defaultValue=5, group=Cliffs_2, parameterName=Cliff Height Lerp Contrast_2}
   NormalTextureObject=@n31
   World Space Normals=@n30
   Export Float 4 Diffuse=@n30
   WorldPosition=AppendVector(A=Add(A=@n34, B=@n35), B=Add(A=@n33, B=@n36))
  ) {materialFunction=&WorldAlignedTextures_Complex}
  n38 = StaticSwitchParameter(True=@n37.XYZ Mat Attributes - See Tooltip, False=@n37.XYZ Mat Attributes - See Tooltip) {defaultValue=true, group=Cliffs_2, parameterName="Use XYZ Cliff_2 Projection. False=XY"}
  n39 = TextureSampleParameter2D() {
   group=Cliffs_2
   parameterName=CliffTexture_2
   samplerSource=SSM_Wrap_WorldGroupSettings
   texture=&T_Cave_Exterior_Wall_D
  }
  n40 = TextureSampleParameter2D() {
   group=Cliffs_2
   parameterName=Cliff Spec Texture_2
   samplerSource=SSM_Wrap_WorldGroupSettings
   texture=&T_Cave_Exterior_Wall_S
  }
  n41 = TextureSampleParameter2D() {
   group=Cliffs_2
   parameterName=CliffNormal_2
   samplerSource=SSM_Wrap_WorldGroupSettings
   samplerType=SAMPLERTYPE_Normal
   texture=&T_Cave_Exterior_Wall_N
  }
  n42 = Transform(None=@n41.RGB)
  n43 = MaterialFunctionCall!(
   Mat attributes=StaticSwitchParameter(
    True=MaterialFunctionCall(In=@n38) {materialFunction=&SwizzleWorldAlignedSpecOutputForFortnite}.Result
    False=MakeMaterialAttributes!(
     BaseColor=@n39.RGB
     Metallic=@n40.G
     Specular=@n40.G
     Roughness=@n40.G
     Anisotropy
     EmissiveColor
     Opacity=@n39.RGB
     OpacityMask
     Normal=@n42
     Tangent
     WorldPositionOffset
     SubsurfaceColor
     ClearCoat
     ClearCoatRoughness
     AmbientOcclusion
     Refraction
     CustomizedUVs_0
     CustomizedUVs_1
     CustomizedUVs_2
     CustomizedUVs_3
     CustomizedUVs_4
     CustomizedUVs_5
     CustomizedUVs_6
     CustomizedUVs_7
     PixelDepthOffset
     ShadingModel
     Displacement
    )
   ) {defaultValue=true, group=Cliffs_2, parameterName=Use World Aligned Cliff 2 Textures}
   Use Spec=StaticBoolParameter() {defaultValue=true, group=Cliffs_2, parameterName=Use Cliff 2 Spec}
   Use Normals=StaticBoolParameter() {defaultValue=true, group=Cliffs_2, parameterName=Use Cliff 2 Normals}
  ) {materialFunction=/Game/Packages/Fortnite_UniqueMaterials_1/MaterialFunctions/Fortnite_Terrain_invertOpacityNormalBool}
  n44 = MaterialFunctionCall!(Mat attributes=@n43.Result) {materialFunction=/Engine/Functions/Engine_MaterialFunctions02/MatLayerBlend_BreakOpacity}
  n45 = LinearInterpolate(A=@n44.Result, Alpha=ScalarParameter() {defaultValue=0.25, group=Cliffs_2, parameterName=Grass/Cliff Transition. Rock or Grass H map lerp_2})
  n46 = MaterialFunctionCall!(
   Base Material
   Top Material=@n43.Result
   Alpha=MaterialFunctionCall(
    Transition Phase=ComponentMask(None=@n27.A) {r=true}
    Height Texture=OneMinus(None=@n45)
    Contrast=ScalarParameter() {defaultValue=2, group=Cliffs_2, parameterName=CliffLerpContrast_2}
   ) {materialFunction=&HeightLerp}.Alpha
  ) {materialFunction=/Engine/Functions/MaterialLayerFunctions/MatLayerBlend_Standard}
  n47 = VertexColor()
  n48 = MaterialFunctionCall(
   Transition Phase=@n47.G
   Height Texture=LinearInterpolate(A=OneMinus(None=@n44.Result), Alpha=@n47.G)
   Contrast=ScalarParameter() {defaultValue=30, group=Cliffs, parameterName=2 Cliff Transition contrast}
  ) {materialFunction=&HeightLerp}
  n49 = TextureSampleParameter2D() {
   group=Cliffs
   parameterName=CliffTexture
   samplerSource=SSM_Wrap_WorldGroupSettings
   texture=&T_Cave_Exterior_Wall_D
  }
  n50 = TextureSampleParameter2D() {
   group=Cliffs
   parameterName=Cliff Spec Texture
   samplerSource=SSM_Wrap_WorldGroupSettings
   texture=&T_Cave_Exterior_Wall_S
  }
  n51 = TextureSampleParameter2D() {
   group=Cliffs
   parameterName=CliffNormal
   samplerSource=SSM_Wrap_WorldGroupSettings
   samplerType=SAMPLERTYPE_Normal
   texture=&T_Cave_Exterior_Wall_N
  }
  n52 = Transform(None=@n51.RGB)
  n53 = StaticSwitchParameter(
   True=StaticSwitchParameter(
    True=MaterialFunctionCall(Base Material=@n46.Blended Material, Alpha=@n48.Alpha) {materialFunction=&MatLayerBlend_Standard}.Blended Material
    False=MaterialFunctionCall(Mat attributes=@n24.Result, Use Spec=@n25, Use Normals=@n26) {materialFunction=&Fortnite_Terrain_invertOpacityNormalBool}.Result
   ) {desc=Diffuse, group=Layers, parameterName="Lerp 2 Cliffs (Uses Green Channel To Lerp)"}
   False=MaterialFunctionCall(
    Mat attributes=MakeMaterialAttributes(BaseColor=@n49.RGB, Metallic=@n50.G, Specular=@n50.G, Roughness=@n50.G, Opacity=@n49.RGB, Normal=@n52, AmbientOcclusion=@n49.RGB)
    Use Spec=StaticBoolParameter() {defaultValue=true, group=Cliffs, parameterName=Use Cliff Spec}
    Use Normals=StaticBoolParameter() {defaultValue=true, group=Cliffs, parameterName=Use Cliff Normals}
   ) {materialFunction=&Fortnite_Terrain_invertOpacityNormalBool}.Result
  ) {group=Layers, parameterName=Use World Aligned Cliff Textures}
  n54 = GetMaterialAttributes(None=@n53) {attributeGetTypes [#1, #2]}
  n55 = ScalarParameter() {group=Cliffs, parameterName=SpecularRoughnessMinCliffs, sortPriority=9}
  n56 = ScalarParameter() {defaultValue=1, group=Cliffs, parameterName=SpecularRoughnessMaxCliffs, sortPriority=8}
  n57 = ComponentMask(None=@n54.Roughness) {r=true}
  n58 = ScalarParameter() {defaultValue=1, group=Cliffs, parameterName=SpecularMultCliffs, sortPriority=10}
  n59 = StaticSwitchParameter(
   True=SetMaterialAttributes!(
    MaterialAttributes=@n54.MaterialAttributes
    Roughness=LinearInterpolate(A=@n55, B=@n56, Alpha=@n57)
    Specular=Multiply(A=@n54.Specular, B=@n58)
   ) {attributeSetTypes [#1, #2]}
   False=@n53
  ) {group=Cliffs, parameterName=CliffsEditSpecTrue}
  n60 = VertexColor()
  n61 = TextureObjectParameter() {group=Dirt, parameterName=Dirt, samplerSource=SSM_Wrap_WorldGroupSettings, texture=/Game/Packages/DS_Fortnight/WhiteRockCustom_D}
  n62 = TextureObjectParameter() {group=Dirt, parameterName=Dirt_SpecMask, samplerSource=SSM_Wrap_WorldGroupSettings, texture=&T_BlacksRGB}
  n63 = TextureObjectParameter() {
   group=Dirt
   parameterName=DirtTextureNormal
   samplerSource=SSM_Wrap_WorldGroupSettings
   samplerType=SAMPLERTYPE_Normal
   texture=/Game/Packages/DS_Fortnight/WhiteRockCustom_N
  }
  n64 = MaterialFunctionCall(
   Diffuse Texture Object=@n61
   Spec Texture Object=@n62
   TextureSize=ScalarParameter() {defaultValue=900, group=Dirt, parameterName=DirtTextureTilingFrequency}
   NormalTextureObject=@n63
   Export Float 4 Diffuse=StaticBool() {value=true}
  ) {materialFunction=&WorldAlignedTextures_Complex}
  n65 = GetMaterialAttributes(None=MaterialFunctionCall(
    Mat attributes=MaterialFunctionCall(In=@n64.Z Mat Attributes - See Tooltip) {materialFunction=&SwizzleWorldAlignedSpecOutputForFortnite}.Result
    Use Spec=StaticBoolParameter() {group=Dirt, parameterName=Use Dirt Spec}
    Use Normals=StaticBoolParameter() {defaultValue=true, group=Dirt, parameterName=Use Dirt Normals}
   ) {materialFunction=&Fortnite_Terrain_invertOpacityNormalBool}.Result) {attributeGetTypes [#3, #4]}
  n66 = Multiply(A=ScalarParameter() {defaultValue=2, group=Dirt, parameterName=DirtBlendX-B}, B=OneMinus(None=@n65.Opacity))
  n67 = ScalarParameter() {defaultValue=5, group=Dirt, parameterName=DirtBlendPower-B}
  n68 = MaterialFunctionCall(Transition Phase=Multiply(A=@n60.B, B=OneMinus(None=@n60.A)), Height Texture=Clamp(None=Power(Base=@n66, Exp=@n67))) {desc=mud + grass blended, materialFunction=&HeightLerp}
  n69 = Power(
   Base=Multiply(A=ScalarParameter() {defaultValue=2, group=Dirt, parameterName=DirtBlendX-A}, B=OneMinus(None=@n65.Opacity))
   Exp=ScalarParameter() {defaultValue=1, group=Dirt, parameterName=DirtBlendPower-A}
  )
  n70 = MaterialFunctionCall(
   In=MaterialFunctionCall(Transition Phase=VertexColor().B, Height Texture=Clamp(None=@n69), Contrast=Constant() {r=1}) {desc=mud + grass blended, materialFunction=&HeightLerp}.Alpha
   Contrast=Constant() {r=1}
  ) {materialFunction=&CheapContrast}
  n71 = VertexColor()
  n72 = MaterialFunctionCall!(Concrete diffuse, Concrete Spec, Concrete Normal) {materialFunction=/Game/Packages/DS_Fortnight/SM/Materials/concreteMF}
  n73 = VertexColor()
  n74 = Clamp(None=OneMinus(None=Multiply(A=@n73.R) {constB=2}) {desc=1- -1})
  n75 = Constant()
  n76 = Add(A=Multiply(B=@n73.R) {constA=-2, constB=-2}, B=OneMinus(None=@n73.A)) {desc=0 to -2}
  n77 = TextureObjectParameter() {
   group=Grass
   parameterName=GrassDiffuse
   samplerSource=SSM_Wrap_WorldGroupSettings
   texture=/Game/Packages/DS_Fortnite_Terrain_NoLOD/Textures/T_Grasslands_D
  }
  n78 = TextureObjectParameter() {
   group=Grass
   parameterName=GrassNormal
   samplerSource=SSM_Wrap_WorldGroupSettings
   samplerType=SAMPLERTYPE_Normal
   texture=/Game/Packages/DS_Fortnite_Terrain_NoLOD/Textures/T_Grasslands_N
  }
  n79 = MaterialFunctionCall(
   Diffuse Texture Object=@n77
   TextureSize=ScalarParameter() {defaultValue=700, group=Grass, parameterName=GrassTiling}
   NormalTextureObject=@n78
   Export Float 4 Diffuse=StaticBool() {value=true}
  ) {materialFunction=&WorldAlignedTextures_Complex}
  n80 = StaticSwitchParameter(True=@n79.XYZ Diffuse Texture, False=@n79.XYZ Diffuse Texture) {group=Grass, parameterName=trueGrassTriplanar}
  n81 = ComponentMask(None=@n80) {b=true, g=true, r=true}
  n82 = Constant()
  n83 = ComponentMask(None=@n80) {a=true}
  n84 = LinearInterpolate(
   B=ScalarParameter() {defaultValue=0.3, group=Grass, parameterName=GrassSpecAmount}
   Alpha=Power(Base=@n83, Exp=ScalarParameter() {defaultValue=2, group=Grass, parameterName=GrassSpecPower})
  )
  n85 = ScalarParameter() {defaultValue=0.8, group=Grass, parameterName=GrassRoughness}
  n86 = OneMinus(None=@n83)
  n87 = TextureObjectParameter() {
   group=BaseTilingTexture
   parameterName=Base Tiling Texture
   samplerSource=SSM_Wrap_WorldGroupSettings
   texture=/Game/Packages/DS_Fortnite_Terrain_NoLOD/Textures/T_FORT_Underground_02
  }
  n88 = TextureObjectParameter() {
   group=BaseTilingTexture
   parameterName=Base Tiling Texture_SpecMask
   samplerSource=SSM_Wrap_WorldGroupSettings
   texture=&T_BlacksRGB
  }
  n89 = TextureObjectParameter() {
   group=BaseTilingTexture
   parameterName=Base Tiling Normal Texture
   samplerSource=SSM_Wrap_WorldGroupSettings
   samplerType=SAMPLERTYPE_Normal
   texture=&FlatNormalBlue
  }
  n90 = MaterialFunctionCall(
   Diffuse Texture Object=@n87
   Spec Texture Object=@n88
   TextureSize=ScalarParameter() {defaultValue=-512, group=BaseTilingTexture, parameterName=Base Tiling Texture Tile Size}
   NormalTextureObject=@n89
   Export Float 4 Diffuse=StaticBool() {value=true}
  ) {materialFunction=&WorldAlignedTextures_Complex}
  n91 = MaterialFunctionCall(In=StaticSwitchParameter(True=@n90.XYZ Mat Attributes - See Tooltip, False=@n90.XYZ Mat Attributes - See Tooltip) {group=BaseTilingTexture, parameterName="Make Base Texture project XYZ(true) or Z(False)"}) {materialFunction=&SwizzleWorldAlignedSpecOutputForFortnite}
  n92 = StaticBoolParameter() {group=BaseTilingTexture, parameterName=Use Base Texture Spec}
  n93 = StaticBoolParameter() {defaultValue=true, group=BaseTilingTexture, parameterName=Use Base Texture Normals}
  n94 = Constant()
  n95 = StaticSwitchParameter(
   True=StaticSwitchParameter(
    True=MakeMaterialAttributes(BaseColor=@n81, Metallic=@n82, Specular=@n84, Roughness=@n85, Opacity=@n86)
    False=MaterialFunctionCall(Mat attributes=@n91.Result, Use Spec=@n92, Use Normals=@n93) {materialFunction=&Fortnite_Terrain_invertOpacityNormalBool}.Result
   ) {defaultValue=true, group=Layers, parameterName="Use Grass(true) or a BaseTilingTexture(False)"}
   False=MakeMaterialAttributes(BaseColor=@n94, Specular=@n94, Opacity=Constant() {r=1})
  ) {defaultValue=true, group=Layers, parameterName=Use a base Layer}
  n96 = MaterialFunctionCall(Mat attributes=@n95) {materialFunction=&MatLayerBlend_BreakOpacity}
  n97 = TextureObjectParameter() {group=Mud, parameterName=Mud_Diffuse, samplerSource=SSM_Wrap_WorldGroupSettings, texture=&WhiteRockCustom_D}
  n98 = TextureObjectParameter() {
   group=Mud
   parameterName=Mud_Normal
   samplerSource=SSM_Wrap_WorldGroupSettings
   samplerType=SAMPLERTYPE_Normal
   texture=&WhiteRockCustom_N
  }
  n99 = MaterialFunctionCall(
   Diffuse Texture Object=@n97
   TextureSize=Abs(None=ScalarParameter() {defaultValue=900, group=Mud, parameterName=MudTextureTilingFrequency})
   NormalTextureObject=@n98
   Export Float 4 Diffuse=StaticBool() {value=true}
  ) {materialFunction=&WorldAlignedTextures_Complex}
  n100 = SetMaterialAttributes!(
   MaterialAttributes
   Base Color=Multiply(
    A=VectorParameter() {group=Mud, parameterName=MudTint, defaultValue=<rgba 0.75 0.5725 0.375 0>}.RGB
    B=ComponentMask(None=@n99.Z Diffuse Texture) {b=true, g=true, r=true}
   )
   Roughness=ScalarParameter() {defaultValue=0.8, group=Mud, parameterName=MudRoughScalar}
   Specular=ScalarParameter() {defaultValue=0.5, group=Mud, parameterName=MudSpecScalar}
   Normal=@n99.Normal Z Texture
   Opacity=ComponentMask(None=@n99.Z Diffuse Texture) {a=true}
  ) {attributeSetTypes [
    #3
    #1
    #2
    #5
    #4
   ]}
  n101 = MaterialFunctionCall(
   Mat attributes=@n100
   Use Spec=StaticBoolParameter() {defaultValue=true, group=Mud, parameterName=Use Mud Spec}
   Use Normals=StaticBoolParameter() {defaultValue=true, group=Mud, parameterName=Use Mud Normals}
  ) {materialFunction=&Fortnite_Terrain_invertOpacityNormalBool}
  n102 = MaterialFunctionCall(Mat attributes=@n101.Result) {materialFunction=&MatLayerBlend_BreakOpacity}
  n103 = MaterialFunctionCall(
   Transition Phase=StaticSwitchParameter(True=StaticSwitchParameter(True=@n74, False=@n75) {defaultValue=true, group=Layers, parameterName=UseConcrete}, False=Clamp(None=@n76)) {group=Layers, parameterName="UseCliffs (Disables Mud)"}
   Height Texture=LinearInterpolate(A=OneMinus(None=@n96.Result), B=@n102.Result, Alpha=ScalarParameter() {defaultValue=0.5, group=Mud, parameterName=MudToGrassBlendA})
   Contrast=ScalarParameter() {defaultValue=1, group=Mud, parameterName=MudToGrassContrast}
  ) {desc=mud + grass blended, materialFunction=&HeightLerp}
  n104 = StaticSwitchParameter(True=@n103.Lerp Alpha No Contrast, False=Constant()) {defaultValue=true, group=Layers, parameterName=UseConcrete}
  n105 = StaticSwitchParameter(
   True=@n103.Lerp Alpha No Contrast
   False=StaticSwitchParameter(True=@n103.Lerp Alpha No Contrast, False=Constant()) {group=Layers, parameterName=UseMud}
  ) {defaultValue=true, group=Layers, parameterName=UseConcrete}
  n106 = OneMinus(None=ComponentMask(None=StaticSwitchParameter(True=@n104, False=@n105) {group=Layers, parameterName="UseCliffs (Disables Mud)"}) {desc=height lerp lerp alpha no contrast from mud, r=true})
  n107 = Clamp(None=@n106)
  n108 = VectorParameter() {group=Mud, parameterName=CustomMud_Sub_Mult_DarkeningLevel_Contrast, defaultValue=<rgba 0.6 0.8 0.2 0>}
  n109 = StaticSwitchParameter(
   True=MaterialFunctionCall(Transition Phase=@n107, Height Texture=@n102.Result, Contrast=@n108.A) {materialFunction=&HeightLerp}.Alpha
   False=@n107
  ) {defaultValue=true, group=Mud, parameterName=Hlerp_MudDarking}
  n110 = MaterialFunctionCall!(Material=@n101.Result, Tint=@n108.B, TintMask=@n109) {materialFunction=/Engine/Functions/MaterialLayerFunctions/MatLayerBlend_Tint}
  n111 = Multiply(
   A=Power(Base=@n109, Exp=ScalarParameter() {defaultValue=2, group=Mud, parameterName=MudWetnessPower})
   B=ScalarParameter() {defaultValue=0.5, group=Mud, parameterName=Mud Wetness Level}
  )
  n112 = StaticSwitchParameter(
   True=MaterialFunctionCall!(Base Material=@n110.Blended Material, RoughnessMultiplier=OneMinus(None=@n111)) {desc=New, materialFunction=/Engine/Functions/MaterialLayerFunctions/MatLayerBlend_ModulateRoughness}.Blended Material
   False=@n110.Blended Material
  ) {group=Mud, parameterName=Use Mud Wetness}
  n113 = ScalarParameter() {defaultValue=0.5, group=Mud, parameterName=GrassMudShadowTintA}
  n114 = ScalarParameter() {defaultValue=2, group=Mud, parameterName=GrassMudShadowPowerA}
  n115 = Clamp(None=Multiply(A=Power(Base=@n106, Exp=@n114), B=ScalarParameter() {defaultValue=1.5, group=Mud, parameterName=GrassMudShadowXA})) {clampMode=CMODE_ClampMax}
  n116 = OneMinus(None=@n115)
  n117 = MaterialFunctionCall(
   Base Material=MaterialFunctionCall!(
    Base Material=MaterialFunctionCall(Material=@n112, Tint=@n113, TintMask=@n116) {materialFunction=&MatLayerBlend_Tint}.Blended Material
    SpecularMultiplier=@n115
   ) {materialFunction=/Engine/Functions/MaterialLayerFunctions/MatLayerBlend_ModulateSpecular}.Blended Material
   Top Material=@n95
   Alpha=@n103.Alpha
  ) {materialFunction=&MatLayerBlend_Standard}
  n118 = StaticSwitchParameter(
   True=StaticSwitchParameter(True=@n117.Blended Material, False=@n95) {defaultValue=true, group=Layers, parameterName=Use_Mud_forced}
   False=StaticSwitchParameter(True=@n117.Blended Material, False=@n95) {group=Layers, parameterName=UseMud}
  ) {defaultValue=true, group=Layers, parameterName=UseConcrete}
  n119 = StaticSwitchParameter(
   True=MaterialFunctionCall(Base Material=@n118, Top Material=@n72.Material Attribute, Alpha=@n72.Material Attribute) {materialFunction=&MatLayerBlend_Standard}.Blended Material
   False=@n118
  ) {defaultValue=true, group=Layers, parameterName=UseConcrete}
  n120 = Multiply(
   A=VectorParameter() {group=Dirt, parameterName=LerpDirtColor, defaultValue=<rgba 0.625 0.5868 0.5375 1>}.RGB
   B=Power(Base=@n65.Base Color, Exp=ScalarParameter() {defaultValue=1, group=Dirt, parameterName=DirtBlendPower-A_1})
  )
  n121 = StaticSwitchParameter(
   True=MaterialFunctionCall(
    Base Material=@n119
    Top Material=SetMaterialAttributes!(MaterialAttributes=@n65.MaterialAttributes, Base Color=@n120) {attributeSetTypes [#3]}
    Alpha=LinearInterpolate(A=@n68.Lerp Alpha No Contrast, B=@n70.Result, Alpha=@n71.B)
   ) {materialFunction=&MatLayerBlend_Standard}.Blended Material
   False=@n119
  ) {defaultValue=true, group=Layers, parameterName=UseDirt}
  n122 = ScalarParameter() {defaultValue=1, group=Cliffs, parameterName=CliffLerpContrast}
  n123 = MaterialFunctionCall(
   Transition Phase=Power(
    Base=VertexColor() {desc=mask cliff to grass transition}.A
    Exp=ScalarParameter() {defaultValue=1.5, group=Cliffs, parameterName=Cliff to Other Layers Transition Power}
   )
   Height Texture=LinearInterpolate(
    A=MaterialFunctionCall(Mat attributes=@n59) {materialFunction=&MatLayerBlend_BreakOpacity}.Result
    B=MaterialFunctionCall(Mat attributes=@n121) {materialFunction=&MatLayerBlend_BreakOpacity}.Result
    Alpha=ScalarParameter() {defaultValue=0.5, group=Cliffs, parameterName=Grass/Cliff Transition. Rock or Grass H map lerp}
   )
   Contrast=@n122
  ) {materialFunction=&HeightLerp}
  n124 = LinearInterpolate(
   B=MaterialFunctionCall(In=@n123.Lerp Alpha No Contrast, Contrast=Multiply(A=@n122) {constB=0.5}) {materialFunction=&CheapContrast}.Result
   Alpha=ScalarParameter() {defaultValue=1.3, group=Cliffs, parameterName=Shadow on Cliff Amount}
  ) {constA=1}
  n125 = TextureObjectParameter() {
   group=TopLayer
   parameterName=Top Lerp Diffuse
   samplerSource=SSM_Wrap_WorldGroupSettings
   texture=/Game/Packages/DS_Fortnight/LeavesDead0019_D
  }
  n126 = TextureObjectParameter() {
   group=TopLayer
   parameterName=Top Lerp Spec Mask
   samplerSource=SSM_Wrap_WorldGroupSettings
   samplerType=SAMPLERTYPE_Masks
   texture=&T_Fortnite_Default_S
  }
  n127 = TextureObjectParameter() {
   group=TopLayer
   parameterName=Top Lerp Normal
   samplerSource=SSM_Wrap_WorldGroupSettings
   samplerType=SAMPLERTYPE_Normal
   texture=&FlatNormalBlue
  }
  n128 = MaterialFunctionCall(
   Diffuse Texture Object=@n125
   Spec Texture Object=@n126
   TextureSize=ScalarParameter() {defaultValue=-512, group=TopLayer, parameterName=Top Layer Tiling Frequency}
   NormalTextureObject=@n127
   Export Float 4 Diffuse=StaticBool() {value=true}
  ) {materialFunction=&WorldAlignedTextures_Complex}
  n129 = TextureSampleParameter2D() {
   group=TopLayer
   parameterName=Top Lerp Diffuse
   samplerSource=SSM_Wrap_WorldGroupSettings
   texture=&LeavesDead0019_D
  }
  n130 = TextureSampleParameter2D() {
   group=TopLayer
   parameterName=Top Lerp Spec Mask
   samplerSource=SSM_Wrap_WorldGroupSettings
   samplerType=SAMPLERTYPE_Masks
   texture=&T_Fortnite_Default_S
  }
  n131 = TextureSampleParameter2D() {
   group=TopLayer
   parameterName=Top Lerp Normal
   samplerSource=SSM_Wrap_WorldGroupSettings
   samplerType=SAMPLERTYPE_Normal
   texture=&FlatNormalBlue
  }
  n132 = Transform(None=@n131.RGB)
  n133 = MaterialFunctionCall(
   Mat attributes=StaticSwitchParameter(
    True=MaterialFunctionCall(In=@n128.Z Mat Attributes - See Tooltip) {materialFunction=&SwizzleWorldAlignedSpecOutputForFortnite}.Result
    False=MakeMaterialAttributes(BaseColor=@n129.RGB, Metallic=@n130.G, Specular=@n130.G, Roughness=@n130.G, Opacity=@n129.RGB, Normal=@n132)
   ) {defaultValue=true, group=TopLayer, parameterName=Use World Aligned Top Layer Textures}
   Use Spec=StaticBoolParameter() {group=TopLayer, parameterName=Use Top Lerp Spec}
   Use Normals=StaticBoolParameter() {group=TopLayer, parameterName=Use Top Lerp Normals}
  ) {materialFunction=&Fortnite_Terrain_invertOpacityNormalBool}
  n134 = MaterialFunctionCall(Material=@n133.Result, Tint=VectorParameter() {group=TopLayer, parameterName=TopLayerColorMultiplier, defaultValue=<rgba 0 0 0 1>}.RGB) {materialFunction=&MatLayerBlend_Tint}
  n135 = VertexColor()
  n136 = ScalarParameter() {defaultValue=0.75, group=TopLayer, parameterName="Top Tiling Mask Texture Raise White Level(0-1)"}
  n137 = TextureObjectParameter() {
   group=TopLayer
   parameterName=Top Tiling Mask Texture
   samplerSource=SSM_Wrap_WorldGroupSettings
   samplerType=SAMPLERTYPE_LinearColor
   texture=/Game/Global/Textures/Default/Noise/T_Noise01_Inverse
  }
  n138 = MaterialFunctionCall(TextureObject=@n137, TextureSize=ScalarParameter() {defaultValue=-1024, group=TopLayer, parameterName=Top Layer Mask Tiling Frequency}) {materialFunction=&WorldAlignedTexture}
  n139 = StaticSwitchParameter(True=OneMinus(None=Multiply(A=@n136, B=@n138.Z Texture) {constB=0.75}), False=@n138.Z Texture) {group=TopLayer, parameterName=Modify Top Tiling Mask Brightness}
  n140 = StaticSwitchParameter(
   True=MaterialFunctionCall(
    Transition Phase=@n135.G
    Height Texture=MaterialFunctionCall(Mat attributes=@n121) {materialFunction=&MatLayerBlend_BreakOpacity}.Result
    Contrast=ScalarParameter() {defaultValue=0.2, group=TopLayer, parameterName=Top Layer Blend Contrast_1}
   ) {materialFunction=&HeightLerp}.Alpha
   False=StaticSwitchParameter(True=Multiply(A=@n135.G, B=@n139), False=@n135.G) {group=TopLayer, parameterName=Modulate Top Level Transition Values With a texture}
  ) {defaultValue=true, group=TopLayer, parameterName=Modulate Top Level Transition Values With LayerOpacity}
  n141 = MaterialFunctionCall(Mat attributes=@n133.Result) {materialFunction=&MatLayerBlend_BreakOpacity}
  n142 = StaticSwitchParameter(True=OneMinus(None=@n141.Result), False=@n141.Result) {defaultValue=true, group=TopLayer, parameterName=InvertTopLayerAlphaChannelMask}
  n143 = ScalarParameter() {defaultValue=256, group=TopLayer, parameterName=Top Layer Blend Contrast}
  n144 = StaticSwitchParameter(
   True=MaterialFunctionCall(
    Base Material=@n121
    Top Material=StaticSwitchParameter(True=@n134.Blended Material, False=@n133.Result) {group=TopLayer, parameterName=Modulate Top Layer Diffuse}
    Alpha=MaterialFunctionCall(Transition Phase=@n140, Height Texture=@n142, Contrast=@n143) {materialFunction=&HeightLerp}.Alpha
   ) {materialFunction=&MatLayerBlend_Standard}.Blended Material
   False=@n121
  ) {group=Layers, parameterName="UseAdditionalTopLayerBlend(ControlledByGreenVertColor)"}
  n145 = StaticSwitchParameter(
   True=MaterialFunctionCall(
    Base Material=@n144
    Top Material=MaterialFunctionCall(Material=@n59, Tint=@n124) {materialFunction=&MatLayerBlend_Tint}.Blended Material
    Alpha=@n123.Alpha
   ) {materialFunction=&MatLayerBlend_Standard}.Blended Material
   False=@n144
  ) {desc=Diffuse, group=Layers, parameterName="UseCliffs (Disables Mud)"}
  n146 = MaterialFunctionCall(
   Base Material=@n145
   Top Material=MakeMaterialAttributes(
    BaseColor=VectorParameter() {group=PaintLayer, parameterName=PaintLayerColor, defaultValue=<rgba 1 1 1 1>}.RGB
    Metallic=ScalarParameter() {group=PaintLayer, parameterName=Paint Metallic}
    Roughness=ScalarParameter() {defaultValue=0.25, group=PaintLayer, parameterName=Paint Roughness}
    Normal=VertexNormalWS()
   )
   Alpha=StaticSwitchParameter(True=Multiply(A=@n12, B=@n13), False=@n12) {desc=Diffuse, group=PaintLayer, parameterName=Modulate Paint Layer Mask}
  ) {materialFunction=&MatLayerBlend_Standard}
  n147 = BreakMaterialAttributes!(Attr=MaterialFunctionCall!(In=StaticSwitchParameter(True=@n146.Blended Material, False=@n145) {desc=Diffuse, group=Layers, parameterName=Use PaintLayer}) {materialFunction=/Game/Athena/WinterArt/Materials/MF_Event_Wet}.Result)
  n148 = ComponentMask(None=Transform(None=@n147.Normal)) {b=true}
  n149 = ScalarParameter() {defaultValue=-0.5143, group=Snow, parameterName=SnowDiffuseContrast1}
  n150 = ScalarParameter() {defaultValue=14.2613, group=Snow, parameterName=SnowDiffuseContrast2}
  n151 = ComponentMask(None=@n147.BaseColor) {r=true}
  n152 = Multiply(A=Clamp(None=LinearInterpolate(A=@n2, B=@n3, Alpha=@n148)), B=Clamp(None=LinearInterpolate(A=@n149, B=@n150, Alpha=@n151)))
  n153 = MaterialFunctionCall!(
   DiffuseFresnelAmount=Constant() {r=0.05}
   Emissive
   Base Color=StaticSwitchParameter(True=LinearInterpolate(A=@n147.BaseColor, B=@n1.RGB, Alpha=@n152), False=@n147.BaseColor) {parameterName=UseSnow}
  ) {materialFunction=/Game/Packages/Fortress_SharedMaterials/Base_Material_Functions/FresnelFunction}
  n154 = Constant()
  n155 = FeatureLevelSwitch(Default=@n147.Metallic, ES2=@n154, ES3_1=@n154, SM4_REMOVED=@n147.Metallic)
  n156 = StaticSwitchParameter(True=@n147.Specular, False=Constant()) {defaultValue=true, group=MasterSettings, parameterName=Use Spec}
  n157 = Constant() {r=0.5}
  n158 = FeatureLevelSwitch(Default=@n156, ES2=@n157, ES3_1=@n157, SM4_REMOVED=@n156)
  n159 = StaticSwitchParameter(True=@n147.Roughness, False=Constant() {r=1}) {defaultValue=true, group=MasterSettings, parameterName=Use Spec}
  n160 = Constant() {r=1}
  n161 = FeatureLevelSwitch(Default=@n159, ES2=@n160, ES3_1=@n160, SM4_REMOVED=@n159)
  n162 = MaterialFunctionCall() {materialFunction=/Game/Packages/Fortnite_UniqueMaterials_1/MaterialFunctions/WorldBorderCliffMask}
  n163 = VertexNormalWS()
  n164 = StaticSwitchParameter(True=QualitySwitch(Default=@n147.Normal, Low=@n163, High=@n147.Normal), False=@n163) {defaultValue=true, group=MasterSettings, parameterName=UseTextureNormals}
  n165 = MaterialFunctionCall!(World Position, DistanceCullFadeAnimationOffset, Use actor center instead of pivot point=StaticBool() {value=true}) {materialFunction=/Game/Packages/Fortress_SharedMaterials/Base_Material_Functions/LODEffect}
  MaterialFunctionCall!(
   In=MaterialFunctionCall(MA In=MakeMaterialAttributes(
     BaseColor=@n153.d
     Metallic=@n155
     Specular=@n158
     Roughness=@n161
     EmissiveColor=@n147.EmissiveColor
     OpacityMask=@n162.Result
     Normal=@n164
     WorldPositionOffset=@n165.Result
     AmbientOcclusion=@n147.EmissiveColor
    )) {materialFunction=&MF_CreativeMoveTool}.MA Out
   UseHeightFalloff=StaticBool() {desc=Do Not UseHeightFalloff here}
   OnFireEmissiveOverride
   HeightFalloffRadius
   HeightFalloffHardness
   VariationMasksMobile
   In_Grass
   Is_Stylised_WE
  ) {materialFunction=/Game/Packages/Fortress_SharedMaterials/Base_Material_Functions/MF_Curie_Landscape}
  StaticSwitchParameter(True=@n79.Normal XYZ Texture, False=@n79.Normal XYZ Texture) {group=Grass, parameterName=trueGrassTriplanar}
  StaticSwitchParameter() {defaultValue=true, group=Layers, parameterName="Use Grass(true) or a BaseTilingTexture(False)"}
 ]
}
